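# Niagara environment effect for Spring Breakout: pink cherry petals drifting off trees, from one GPU emitter with sprite and mesh renderers.
# Spawn points in a sphere snap to the depth buffer, are culled by validity and displacement, then filtered by a pink GBuffer hue range.
# Gravity, curl noise, wind, drag and collision drive motion.
# Distance scalability, a particle-count gate, camera-distance sizing and time-of-day material parameters control cost and look.
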NiagaraSystem /Game/Effects/Fort_Effects/Effects/Environment/Cherry/NS_Env_Cherry_FromTrees_SpringBreakout
dependencies {
 usedDataInterfaces [
  &NiagaraDataInterfaceCamera
  &NiagaraDataInterfaceCurve
  &NiagaraDataInterfaceGBuffer
  &NiagaraDataInterfaceArrayFloat
  &NiagaraDataInterfaceCollisionQuery
 ]
 usedDynamicInputs []
 usedModules [
  &SystemState
  &self:SystemSpawnScript.SystemScriptSource.SystemScriptGraph.NiagaraNodeAssignment_0.SetVariables_#1
 ]
}
emitters {LeavesInAir {
  simTarget=GPUComputeSim
  emitterUpdateScript [
   /SetVariables(
    Emitter.CameraPostion:Vector3f=&ReturnCameraVector(Camera Property=::NewEnumerator0, Coordinate Space=::Simulation)
    Emitter.Emitter_HeightOffset:NiagaraFloat=$System.System_HeightOffset
    Emitter.Overall_SpeedMult=0.2
    Emitter.SpawnRate_Curve:NiagaraFloat=&FloatFromCurve(
     FloatCurve=DataInterface() {propertyValues {Curve {keys [
         {arriveTangent=17.5837, interpMode=RCIM_Cubic, tangentMode=RCTM_User, time=0, value=0}
         {arriveTangent=2.6573, interpMode=RCIM_Cubic, leaveTangent=2.6573, tangentMode=RCTM_User, time=0.1546, value=0.275}
         {interpMode=RCIM_Cubic, tangentMode=RCTM_User, time=0.3, value=1}
         <at 0.7 0>
         <at 0.85 0>
        ]}}}
     CurveIndex=$Emitter.NormalizedLoopAge
     Scale Curve=100
    )
    Emitter.SpawnSpeed_Curve:NiagaraFloat=&FloatFromCurve(
     FloatCurve=DataInterface() {propertyValues {Curve {keys [
         {arriveTangent=30.2685, interpMode=RCIM_Cubic, leaveTangent=30.2685, tangentMode=RCTM_User, time=0.0009, value=0.0215}
         {arriveTangent=-0.0619, interpMode=RCIM_Cubic, leaveTangent=-0.0619, tangentMode=RCTM_User, time=0.0989, value=0.3515}
         <at 0.297 0.0032>
         <at 0.3192 0.3332>
         <at 0.3525 0.0398>
         <at 0.6171 0.3148>
         <at 0.9939 0.3698>
        ]}}}
     CurveIndex=&Multiply_Float(A=$Emitter.NormalizedLoopAge, B=&RandomRangeFloat(Minimum=1, Maximum=5))
     Scale Curve=&Multiply_Float(
      A=&Lerp_Float(A=5, B=10, Alpha=&UniformRangedFloat(Minimum=0.25, Maximum=1))
      B=1
     )
    )
    Emitter.SpriteRotRate:NiagaraFloat=&FloatFromCurve(
     FloatCurve=DataInterface() {propertyValues {Curve {keys [
         <at 0.0041 0.1315>
         <at 0.1088 1.9664>
         {arriveTangent=-3.3757, interpMode=RCIM_Cubic, leaveTangent=-3.3757, tangentMode=RCTM_User, time=0.3609, value=0.6391}
         {arriveTangent=-1, interpMode=RCIM_Cubic, leaveTangent=-1, tangentMode=RCTM_User, time=0.5614, value=0.4386}
         <at 0.7086 1.6472>
         <at 1 0>
        ]}}}
     CurveIndex=$Emitter.NormalizedLoopAge
     Scale Curve=500
    )
   ) {isSetParameters=true}
   &EmitterState(
    Enable Distance Culling=true
    Enable Visibility Culling=true
    MaxDistance=10000
    Min Distance Response=::SleepAndLetParticlesFinish
    MinDistance=1250
    Module.bApplySpawnCountScaleByDistance=true
    Module.bUseMaxDistance=true
    Module.bUseMinDistance=true
    Scalability Mode=::NewEnumerator1
    Scale Spawn Count=true
    Spawn Count Scale By Distance=DataInterface() {propertyValues {Curve {keys [<at 0 0>,<at 0.1 0.25>,<at 0.2 0.35>,<at 0.8 0.5>,<at 1 0>]}}}
    Visibility Cull Delay=10
   )
   &SpawnRate(SpawnRate=2500, Use Spawn Probability=false) {enabled=false, module=SpawnRate001}
   &SpawnRate(
    SpawnRate=&Multiply_Float(
     A=$Emitter.SpawnRate_Curve
     B=&MakeCustomFloatFromBool(
      Bool=&SetBoolByIntComparison(A=$Engine.Emitter.NumParticles, B=1500, Comparison Type=::NewEnumerator0)
      True Float=0
      False Float=500
     )
    )
    Use Spawn Probability=false
   )
  ]
  particleSpawnScript [
   /SetVariables(
    Particles.HeightOffset=200
    Particles.UporDown:NiagaraFloat=&UniformAOrBFloat(
     A=&RandomRangeFloat(Minimum=12, Maximum=15)
     B=&RandomRangeFloat(Minimum=250, Maximum=1500)
     Distribution Weight=0.8
    )
   ) {isSetParameters=true}
   &InitializeParticle(
    Color=&RandomRangeLinearColor(Minimum=<rgba 0.7011 0.482 0.528 1>, Maximum=<rgba 0.9219 0.4493 0.3649 1>)
    Lifetime=&RandomRangeFloat(Minimum=3, Maximum=5)
    Mesh Scale Mode=::NewEnumerator1
    Mesh Uniform Scale Max=0.0175
    Mesh Uniform Scale Min=0.0135
    Sprite Size Mode=::NewEnumerator1
    Uniform Sprite Size Max=15
    Uniform Sprite Size Min=12
    UsePositionOffset=false
    Write Material Random=true
   ) {module=InitializeParticle001}
   /Niagara/Modules/Spawn/Location/SphereLocation(
    MaskSpawnGroup=false
    Module.Override Randomness=false
    Non Uniform Scale=&RandomRangeVector(Minimum=<-1 -1 -0.5>, Maximum=<1 1 1>)
    Sphere Radius=&Multiply_Float(A=800, B=1)
    Surface Only=false
   ) {module=SphereLocation001}
   /SetVariables(Particles.InitialPos:NiagaraPosition=$Particles.Position) {isSetParameters=true}
   /Niagara/Modules/Math/PlaceParticlesOnDepthBuffer_GPU(Max Z Depth=10000, Module.Use Max Z Depth=true)
   /Game/Effects/Niagara/Modules/Spawn/FN_Spawn_GBuffer_TransparentOnly(AmountOfRedMax=1, AmountOfRedMin=1) {enabled=false}
   &KillParticles(Kill Particles=&SetBoolByFloatComparison(
     A=&VectorLength(Vector=&Subtract_Vector(A=$Particles.InitialPos, B=Linked() {linkedVariable {name=Particles.Position, type=NiagaraPosition}}))
     B=500
    ))
   &KillParticles(Kill Particles=&Boolean_NotOperation(Bool=$Output.PlaceParticlesOnDepthBuffer_GPU.ValidSample)) {module=KillParticles001}
   /Game/Effects/Niagara/Modules/Spawn/FN_Spawn_GBuffer_Color_WithinRangeV2(HueMax=335, HueMin=329)
   &AddVelocity(Velocity=&Multiply_VectorByFloat(
     Vector=&Multiply_VectorByFloat(
      Vector=&Multiply_VectorByFloat(Vector=&UniformRangedVector(Minimum=<30 150 -20>, Maximum=<80 150 0>), Float=$Emitter.SpawnSpeed_Curve)
      Float=$Emitter.Overall_SpeedMult
     )
     Float=1
    ))
   &CurlNoiseForce(
    Noise Frequency=2
    Noise Strength=&Multiply_Float(
     A=800
     B=&Multiply_Float(A=&Multiply_Float(A=1, B=$Emitter.Overall_SpeedMult), B=1)
    )
    Pan Noise=false
   ) {module=CurlNoiseForce001}
   &ApplyInitialForces()
   &CameraOffset(Camera Offset Amount=5)
   &InitialMeshOrientation(Use Rotation Vector=false)
  ]
  particleUpdateScript [
   /SetVariables(Particles.ParticleUpdate_RotRate:NiagaraFloat=&FloatFromCurve(
     FloatCurve=DataInterface() {propertyValues {Curve {keys [
         <at 0.0041 0.1315>
         <at 0.1088 1.9664>
         {arriveTangent=-3.3757, interpMode=RCIM_Cubic, leaveTangent=-3.3757, tangentMode=RCTM_User, time=0.3609, value=0.6391}
         {arriveTangent=-1, interpMode=RCIM_Cubic, leaveTangent=-1, tangentMode=RCTM_User, time=0.5614, value=0.4386}
         <at 0.7086 1.6472>
         {arriveTangent=-1.004, interpMode=RCIM_Cubic, leaveTangent=-1.004, tangentMode=RCTM_User, time=0.8311, value=0.9545}
         {arriveTangent=-5.6531, interpMode=RCIM_Cubic, leaveTangent=-5.6531, tangentMode=RCTM_User, time=0.8987, value=1.4563}
         <at 0.9384 0.3483>
         <at 1 0>
        ]}}}
     Scale Curve=100
    )) {isSetParameters=true}
   &ParticleState()
   &ScaleMeshSize(Scale Factor=&VectorFromFloat(Value=&MakeFloatFromVector2d(Vector 2D=&Vector2DFromFloat(Value=&Multiply_Float(
        A=&FloatFromCurve(FloatCurve=DataInterface() {propertyValues {Curve {keys [{arriveTangent=0.9503, time=0, value=0}, <at 0.1 1>, {arriveTangent=-1.2812, time=0.9, value=1}, <at 1 0>]}}})
        B=&FloatFromCurve(
         FloatCurve=DataInterface() {propertyValues {Curve {keys [
             {interpMode=RCIM_Cubic, tangentMode=RCTM_User, time=0, value=0.5}
             {arriveTangent=0.0074, leaveTangent=0.0074, time=0.15, value=1.0236}
             {arriveTangent=-0.0952, leaveTangent=-0.0952, time=0.85, value=1.35}
             <at 0.9 0>
            ]}}}
         CurveIndex=&NormalizeDistanceRange(
          Start Position:NiagaraPosition=Linked() {linkedVariable {name=Emitter.CameraPostion, type=Vector3f}}
          End Position:NiagaraPosition=$Particles.Position
          Distance:NiagaraFloat=10000
          Invert Normalized Range:NiagaraBool=false
         )
        )
       )))))
   /Niagara/Modules/Update/SubUV/SubUVAnimation(Number Of Frames=64)
   &ScaleSpriteSize(Scale Factor=&Multiply_Vector2(
     A=<1 1>
     B=&Vector2DFromFloat(Value=&Multiply_Float(
       A=&FloatFromCurve(FloatCurve=DataInterface() {propertyValues {Curve {keys [{arriveTangent=0.9503, time=0, value=0}, <at 0.1 1>, {arriveTangent=-1.2812, time=0.9, value=1}, <at 1 0>]}}})
       B=&FloatFromCurve(
        FloatCurve=DataInterface() {propertyValues {Curve {keys [
            {interpMode=RCIM_Cubic, tangentMode=RCTM_User, time=0, value=0.25}
            {arriveTangent=0.0074, interpMode=RCIM_Cubic, leaveTangent=0.0074, tangentMode=RCTM_User, time=0.15, value=1.0236}
            {arriveTangent=-0.0952, interpMode=RCIM_Cubic, leaveTangent=-0.0952, tangentMode=RCTM_User, time=0.85, value=1}
            <at 0.9 0>
           ]}}}
        CurveIndex=&NormalizeDistanceRange(
         Start Position:NiagaraPosition=Linked() {linkedVariable {name=Emitter.CameraPostion, type=Vector3f}}
         End Position:NiagaraPosition=$Particles.Position
         Distance:NiagaraFloat=10000
         Invert Normalized Range:NiagaraBool=false
        )
        Scale Curve=0.85
       )
      ))
    )) {module=ScaleSpriteSize001}
   &SpriteRotationRate(Rotation Rate=$Emitter.SpriteRotRate)
   &ScaleColor(
    Scale Alpha=&FloatFromCurve(
     FloatCurve=DataInterface() {propertyValues {Curve {keys [
         {interpMode=RCIM_Cubic, tangentMode=RCTM_User, time=0, value=0}
         {arriveTangent=-0.0948, interpMode=RCIM_Cubic, leaveTangent=-0.0948, tangentMode=RCTM_User, time=0.1, value=1}
         {interpMode=RCIM_Cubic, tangentMode=RCTM_User, time=0.7, value=1.05}
         {interpMode=RCIM_Cubic, tangentMode=RCTM_User, time=1, value=0}
        ]}}}
     Scale Curve=0.9
    )
    Scale RGB=&MakeVectorFromLinearColorRGB(Color=&HueShiftLinearColor(
      Hue=&RandomRangeFloat(Minimum=0, Maximum=0.02)
      Saturation=1
      Value=&RandomRangeFloat(Minimum=1, Maximum=1)
      Linear Color=$Particles.Initial.Color
     ))
    ScaleA=true
    ScaleRGB=true
   )
   /Niagara/Modules/Update/SubUV/V2/SubUVAnimation(
    Start Frame:NiagaraInt32=&RandomRangeInt(Minimum=0, Maximum=32)
    SubUV Animation Mode:ENiagara_SubUVLookupMode=::NewEnumerator1
    SubUV Loop Count:NiagaraFloat=&SelectFloatFromArray(
     Array Sampling Mode=::NewEnumerator0
     Direct Array Index=?
     Float Selection Array=DataInterface() {propertyValues {FloatData [2,4]}}
     Random Seed=0
    )
   ) {module=SubUVAnimation001}
   &GravityForce(Gravity=<0 0 -50>)
   &DynamicMaterialParameters(
    Index 0 Param 1=&FN_Select_Float_from_Time_Of_Day(Morning=0.05, Day=0.25, Evening=0.1, Night=0.15)
    Module.Param0WriteEnabled=true
    Module.Param1WriteEnabled=true
    Module.Param2WriteEnabled=true
    Module.Param3WriteEnabled=true
   )
   &WindForce(
    Frequency Min / Max:Vector2f=<0.1 1>
    Friction Distance Limit=2500
    Offset By Time:Vector3f=<0 0 0>
    Scale Curve=DataInterface() {propertyValues {Curve {keys [{tangentMode=RCTM_User, time=0, value=1}, {tangentMode=RCTM_User, time=1, value=1}]}}}
    Scale Min / Max:Vector2f=<1 2>
    Speed Min / Max:Vector2f=<0 1000>
    Turbulence Frequency Mode:ENiagaraWindTurbulenceFrequencyMode=::NewEnumerator1
    Turbulence Scale Mode:ENiagaraWindTurbulenceContributionMode=::NewEnumerator3
    Wind Speed=<150 0 -25>
   ) {module=WindForce002}
   &AerodynamicDrag(
    Aerodynamic Drag=0.05
    Aerodynamic Rotational Drag=0.1
    Pivot Offset Maximum:Vector3f=<0.5 0.5 0.5>
    Pivot Offset Minimum:Vector3f=<-1 -1 -1>
    Pivot Offset Minimum Length:Vector3f=<0.15 0.15 0.15>
    Rotation Strength:NiagaraFloat=&RandomRangeFloat(Minimum=3, Maximum=10)
    Use Linear Drag=true
    Use Rotational Drag=true
   )
   &Collision(
    Age Colliding Particles=false
    Collision Enabled=&SetBoolByFloatComparison(A=0.2, B=$Particles.NormalizedAge, Comparison Type=::NewEnumerator7)
    Control Roll On Collision=true
    Friction=0.5
    Friction During a Bounce=0.25
    GPU Collision Type=::NewEnumerator2
    Module.Control Rest State Via Surface Normal=true
    Module.EnableMaxCollisionCount=false
    Module.Randomize Collision Normal Vector=false
    Radius Calculation Type=::NewEnumerator1
    Restitution=0
    ScaleRateOfRotation=true
   ) {module=Collision001}
   &SolveForcesAndVelocity(Clamp Velocity=true, Limit Acceleration=false, Speed Limit=250)
   &AlignSpriteToMeshOrientation(SetAlignment=true, SetFacing=true)
  ]
  rendererClasses [&NiagaraSpriteRendererProperties, &NiagaraMeshRendererProperties]
  renderers [
   {rendererClass=&NiagaraSpriteRendererProperties}
   {rendererClass=&NiagaraMeshRendererProperties, rendererIndex=1}
  ]
 }}
summary {
 systemName=NS_Env_Cherry_FromTrees_SpringBreakout
 emitters [{
   emitterName=LeavesInAir
   simTarget=GPUComputeSim
   rendererClasses [&NiagaraSpriteRendererProperties, &NiagaraMeshRendererProperties]
  }]
 userVariables [User.Landscape {
   type=NiagaraDataInterfaceLandscape
   description=""
   defaultValue {
    struct=NiagaraExt_VariableValue_DataInterface
    value {
     dataInterface=&self:NiagaraDataInterfaceLandscape_0
     dataInterfaceClass=&NiagaraDataInterfaceLandscape
    }
   }
  }]
}
